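# Sound cue for the first-person firing sound of the suppressed assault rifle.
# A random node picks one of four equally weighted fire waves, then a modulator slightly varies pitch and volume to reduce repetition.
# It uses first-person weapon attenuation, a ranged-fire sound class and a gunfire concurrency set, alongside overrides for an owner-limited, stop-oldest voice cap.
# The debug flag is left enabled.
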SoundCue /Game/Sounds/Fort_Weapons/AutoRifle/SuppressedAssault/Suppressed_Assault_Fire_1P_Cue
root=Modulator_0
nodes [
 {id=Modulator_0, children [Random_5], properties {pitchMax=1.03, pitchMin=0.97, volumeMax=1, volumeMin=0.9}}
 {id=Random_5, children [WavePlayer_4,WavePlayer_5,WavePlayer_6,WavePlayer_7], properties {weights [1,1,1,1]}}
 {id=WavePlayer_4, properties {soundWaveAssetPtr=&/Suppressed_Assault_FireC-001}}
 {id=WavePlayer_5, properties {soundWaveAssetPtr=&/Suppressed_Assault_FireC-002}}
 {id=WavePlayer_6, properties {soundWaveAssetPtr=&/Suppressed_Assault_FireC-003}}
 {id=WavePlayer_7, properties {soundWaveAssetPtr=&/Suppressed_Assault_FireC-004}}
]
properties {
 attenuationSettings=&Fort_Weapon_Projectile_FirstPerson_Athena_Att
 bDebug=true
 maxDistance=2097152
 soundClassObject=&Fort_WeaponRanged_Fire_1P_PSM
 volumeMultiplier=0.6
 concurrencyOverrides {bLimitToOwner=true, maxCount=3, resolutionRule=StopOldest}
 concurrencySet [&Weapon_Gunfire_FirstPerson_Conc]
}
